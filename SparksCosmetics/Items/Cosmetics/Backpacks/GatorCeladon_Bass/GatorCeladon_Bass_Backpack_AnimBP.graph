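# Animation Blueprint for the GatorCeladon bass back bling on the generic bass skeleton, shifting the instrument's position while skydiving.
# Skydive begin and end events set the IsSkydiving flag, which drives the alpha of a Modify Bone node.
# That node applies a fixed additive, bone-space translation and roll to the offset bone of the input pose.
# The BackpackOffset variables are declared but unused.
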AnimBlueprint /SparksCosmetics/Items/Cosmetics/Backpacks/GatorCeladon_Bass/GatorCeladon_Bass_Backpack_AnimBP
target_skeleton=/SparksCharacterCommon/Instrument/Bass/Generic/Meshes/Generic_Bass_Skeleton
graphs [
 AnimGraph {nodes [
   n1 = "Transform (Modify) Bone - Bone: offset"(
    ComponentPose=Local To Component(LocalPose=Input Pose())
    Alpha="To Float (Boolean)"(InBool=Get IsSkydiving() -> (IsSkydiving:bool))
    Translation="2.5,0,4"
    Rotation="0,0,4"
    Scale="1,1,1"
   ) {
    rotationMode=BMM_Additive
    rotationSpace=BCS_BoneSpace
    scaleMode=BMM_Additive
    scaleSpace=BCS_BoneSpace
    translationMode=BMM_Additive
    translationSpace=BCS_BoneSpace
    alphaBoolBlend {blendOption=Linear}
    boneToModify {boneName=offset}
    rotation {roll=4}
    translation {x=2.5, z=4}
   }
   Output Pose(Result=Component To Local(ComponentPose=@n1))
  ]}
 EventGraph {nodes [
   Event OnBeginSkydiving()
   > Set IsSkydiving(IsSkydiving:bool=true) -> (Output_Get:bool)
   Event OnEndSkydiving()
   > Set IsSkydiving(IsSkydiving=false)
   TryGetPawnOwner()
   Event BlueprintUpdateAnimation()
  ]}
]
variables [
 BackpackOffsetLocation {type=Vector}
 BackpackOffsetRotation {type=Rotator}
 BackpackOffsetScale {type=Vector}
 IsSkydiving {type=boolean}
]
